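# Default Lit material for Minerva creative prop-set statues, with parameterized color, gloss-to-matte finish and emissive glow.
# Base color optionally blends in a second-UV color map, falling back to the flat color where flipped V is below a threshold, then gets a contrast-adjusted Fresnel tint.
# Vertex color red blends gloss and matte values for roughness and specular.
# Emissive sums a flat color and a separate Fresnel rim term.
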Material /Game/Creative/Sets/PropSets/Minerva/Materials/M_CP_Minerva_Statue
graph {
 properties {pixelDepthOffsetMode=PDOM_Legacy, usedShadingModels=Default Lit}
 nodes [
  n1 = VectorParameter() {group=Emissive Fresnel, parameterName=Emissive Fresnel Color, sortPriority=1, defaultValue=<rgba 1 1 1 1>}
  n2 = MaterialFunctionCall(
   In=Fresnel(ExponentIn=ScalarParameter() {defaultValue=1, group=Emissive Fresnel, parameterName=Emissive Fresnel Exp, sortPriority=2})
   Contrast=ScalarParameter() {group=Emissive Fresnel, parameterName=Emissive Fresnel Contrast, sortPriority=3}
  ) {materialFunction=&CheapContrast}
  n3 = ScalarParameter() {group=Emissive Fresnel, parameterName=Emissive Fresnel Intensity}
  Add(
   A=Multiply(
    A=VectorParameter() {group=Emissive, parameterName=Emissive Color, defaultValue=<rgba 0 0 0 1>}.RGB
    B=ScalarParameter() {group=Emissive, parameterName=Emissive Intensity}
   )
   B=Multiply(A=@n1.RGB, B=Multiply(A=@n2.Result, B=@n3))
  )
  Constant()
  n4 = VertexColor()
  LinearInterpolate(
   A=ScalarParameter() {defaultValue=0.5, parameterName=Gloss Roughness}
   B=ScalarParameter() {defaultValue=0.6, parameterName=Matte Roughness}
   Alpha=@n4.R
  )
  LinearInterpolate(
   A=ScalarParameter() {defaultValue=1, parameterName=Gloss Specular}
   B=ScalarParameter() {defaultValue=0.8, parameterName=Matte Specular}
   Alpha=@n4.R
  )
  n5 = VectorParameter() {group=Color, parameterName=Color, sortPriority=1, defaultValue=<rgba 0.015 0.015 0.015 1>}
  n6 = TextureCoordinate() {coordinateIndex=1}
  n7 = OneMinus(None=MaterialFunctionCall(Float2=@n6) {materialFunction=&BreakOutFloat2Components}.G)
  n8 = ScalarParameter() {
   defaultValue=0.01
   desc=UVs at this V value and below will use the Color Parameter value and not the color texture map
   group=Color
   parameterName=UVSet2 Mask Range
  }
  n9 = TextureSampleParameter2D(UVs=@n6) {group=Color, parameterName=Color Map UVSet2, sortPriority=3, texture=/Game/Creative/BuildingActors/Props/Textures/CP_White_Texutre}
  n10 = Fresnel(ExponentIn=ScalarParameter() {defaultValue=1, group=Color Fresnel, parameterName=Fresnel Exp, sortPriority=2})
  n11 = ScalarParameter() {group=Color Fresnel, parameterName=Fresnel Contrast, sortPriority=3}
  LinearInterpolate(
   A=LinearInterpolate(
    A=@n5.RGB
    B=If(A=@n7, B=@n8, "A > B"=@n9.RGB, "A == B"=@n9.RGB, "A < B"=@n5.RGB)
    Alpha=ScalarParameter() {group=Color, parameterName=Color with UVSet2, sortPriority=2}
   )
   B=VectorParameter() {group=Color Fresnel, parameterName=Fresnel Color, sortPriority=1, defaultValue=<rgba 1 1 1 1>}.RGB
   Alpha=Multiply(
    A=MaterialFunctionCall(In=@n10, Contrast=@n11) {materialFunction=&CheapContrast}.Result
    B=ScalarParameter() {group=Color Fresnel, parameterName=Fresnel Intensity}
   )
  )
  TextureSampleParameter2D() {
   group=Normal
   parameterName=Normal Map
   samplerType=SAMPLERTYPE_Normal
   sortPriority=10
   texture=/Engine/EngineMaterials/BaseFlattenNormalMap
  }
  TextureSampleParameter2D() {
   group=Specular
   parameterName=SpecularMasks
   samplerSource=SSM_Wrap_WorldGroupSettings
   samplerType=SAMPLERTYPE_Masks
   texture=&T_Fortnite_Default_S
  }
 ]
}
